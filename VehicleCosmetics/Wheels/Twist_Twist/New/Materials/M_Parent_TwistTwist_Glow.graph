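# Parent glow material, by its name, for the Twist_Twist wheel vehicle cosmetic.
# The graph holds no expressions, so the file shows no glow logic, parameters or connected inputs.
# It uses the Default Lit shading model and has automatic setting of usage flags in the editor turned off.
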Material /VehicleCosmetics/Wheels/Twist_Twist/New/Materials/M_Parent_TwistTwist_Glow
graph {expressions {}, properties {bAutomaticallySetUsageInEditor=false, usedShadingModels=Default Lit}}
